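# Physics asset for the Space Warrior male axe pickaxe, giving capsule bodies to the two-handed handle and the weapon's dynamic lever, back component, base component and cord bones.
# The base component hinges on the handle with one limited swing axis, and the lever and back component hinge likewise on the base.
# The cord's middle bone has narrow swing and twist limits against the base and a free-rotation constraint to the back component, closing a loop.
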PhysicsAsset /BRCosmetics/Weapons/FORT_Melee/Pickaxe_Space_Warrior/Meshes/Space_Warrior_Male_Axe_Physics
bodies {
 Melee_TwoHanded_Handle [{shapeType=Capsule, length=50, radius=3, center=<0 -1.2344 9.9619>, rotation=<rot -0.073 -90 0.0203>}]
 dyn_backComponent_rotate_jnt [{shapeType=Capsule, length=13.5694, radius=8.2062, center=<1.5663 0.0128 3.396>, rotation=<rot 0 -89.9095 -69.2255>}]
 dyn_baseComponent_rotate_jnt [{shapeType=Capsule, length=40, radius=15, center=<27.1656 7.8198 0.0876>, rotation=<rot -0.0081 -167.3301 -89.9708>}]
 dyn_cord_mid_01 [{shapeType=Capsule, length=5, radius=5, center=<5.773 0 0>, rotation=<rot 0 -90 0>}]
 dyn_lever_hinge_jnt [{shapeType=Capsule, length=21.8339, radius=3, center=<12.1549 0.0005 -10.1046>, rotation=<rot 8.9871 0 0>}]
}
constraints [
 {
  bone1Name=dyn_lever_hinge_jnt
  bone2Name=dyn_baseComponent_rotate_jnt
  swing1LimitDegrees=45
  swing1Motion=Locked
  swing2LimitDegrees=30
  twistLimitDegrees=45
  twistMotion=Locked
 }
 {
  bone1Name=dyn_baseComponent_rotate_jnt
  bone2Name=Melee_TwoHanded_Handle
  swing1LimitDegrees=25
  swing2LimitDegrees=45
  swing2Motion=Locked
  twistLimitDegrees=45
  twistMotion=Locked
 }
 {
  bone1Name=dyn_backComponent_rotate_jnt
  bone2Name=dyn_baseComponent_rotate_jnt
  swing1LimitDegrees=45
  swing1Motion=Locked
  swing2LimitDegrees=10
  twistLimitDegrees=45
  twistMotion=Locked
 }
 {bone1Name=dyn_cord_mid_01, bone2Name=dyn_baseComponent_rotate_jnt, swing1LimitDegrees=20, swing2LimitDegrees=20, twistLimitDegrees=5}
 {
  bone1Name=dyn_backComponent_rotate_jnt
  bone2Name=dyn_cord_mid_01
  swing1LimitDegrees=45
  swing1Motion=Free
  swing2LimitDegrees=45
  swing2Motion=Free
  twistLimitDegrees=45
  twistMotion=Free
 }
]
