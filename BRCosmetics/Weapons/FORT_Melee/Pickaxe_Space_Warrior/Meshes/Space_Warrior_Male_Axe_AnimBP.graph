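# Animation Blueprint for the male Space Warrior axe pickaxe, animating its chain teeth.
# A sped-up chain animation sequence is layered per bone over the input pose on the dyn_tooth_jnt branches and drives the output.
# A RigidBody physics node in base-bone space with damped inherited motion is present but not connected to the output pose.
# The event graph holds only unconnected stub nodes.
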AnimBlueprint /BRCosmetics/Weapons/FORT_Melee/Pickaxe_Space_Warrior/Meshes/Space_Warrior_Male_Axe_AnimBP
target_skeleton=&/Space_Warrior_Male_Axe_Skeleton
graphs [
 AnimGraph {nodes [
   n1 = RigidBody(ComponentPose=Local To Component(), Alpha=0.9) {
    bClampLinearTranslationLimitToRefPose=true
    bOverrideWorldGravity=true
    overridePhysicsAsset=&/Space_Warrior_Male_Axe_Physics
    simulationSpace=BaseBoneSpace
    alphaBoolBlend {blendOption=Linear}
    baseBoneRef {boneName=b_AR_Root}
    componentAppliedLinearAccClamp=<6000 6000 6000>
    componentLinearAccScale=<0.25 0.25 0.25>
    componentLinearVelScale=<0.3 0.3 0.3>
   }
   Component To Local(ComponentPose=@n1)
   n2 = Layered blend per bone(
    BasePose=Input Pose()
    BlendPoses_0=Sequence Player 'Space_Warrior_Male_Axe_ChainAnimation'() {playRate=3, sequence=&/Space_Warrior_Male_Axe_ChainAnimation}
   ) {layerSetup [{branchFilters [
       {boneName=dyn_tooth_jnt_1}
       {boneName=dyn_tooth_jnt_2}
       {boneName=dyn_tooth_jnt_3}
       {boneName=dyn_tooth_jnt_4}
       {boneName=None}
       {boneName=None}
       {boneName=None}
       {boneName=None}
       {boneName=None}
       {boneName=None}
       {boneName=None}
       {boneName=None}
       {boneName=None}
       {boneName=None}
       {boneName=None}
       {boneName=None}
       {boneName=None}
       {boneName=None}
       {boneName=None}
       {boneName=None}
       {boneName=None}
       {boneName=None}
       {boneName=None}
       {boneName=None}
       {boneName=None}
       {boneName=None}
       {boneName=None}
       {boneName=None}
       {boneName=None}
       {boneName=None}
       {boneName=None}
       {boneName=None}
       {boneName=None}
       {boneName=None}
       {boneName=None}
      ]}]}
   Output Pose(Result=@n2) {layerGroup=None}
  ]}
 EventGraph {nodes [
   TryGetPawnOwner()
   Event BlueprintUpdateAnimation()
  ]}
]
variables []
